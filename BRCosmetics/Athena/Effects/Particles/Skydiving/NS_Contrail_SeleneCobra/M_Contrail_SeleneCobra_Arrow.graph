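# Unlit translucent, two-sided Niagara mesh-particle material for the SeleneCobra skydiving contrail, drawing an arrow with a glowing trail.
# Vertex red selects flat arrow color over a particle-tinted trail gradient with sand texture and bright center line.
# Trail opacity combines UV falloff masks with two panning noise samples.
# Vertex blue and inverted V mask vertex displacement, with dynamic parameters stretching the trail along Z and scaling it from the object pivot.
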Material /BRCosmetics/Athena/Effects/Particles/Skydiving/NS_Contrail_SeleneCobra/M_Contrail_SeleneCobra_Arrow
graph {
 properties {
  bUsedWithNiagaraMeshParticles=true
  blendMode=BLEND_Translucent
  pixelDepthOffsetMode=PDOM_Legacy
  shadingModel=MSM_Unlit
  twoSided=true
  usedShadingModels=Unlit
 }
 nodes [
  n1 = TextureObjectParameter() {
   group=SandA
   parameterName=SandA_Texture
   samplerType=SAMPLERTYPE_Normal
   sortPriority=0
   texture=/Game/Tools/WaterfallGenerator/Art/T_Waterfall_Sand
  }
  n2 = Constant3Vector() {constant=<rgba 0.25 1 1 0>}
  n3 = Saturate(None=Multiply(
    A=MaterialFunctionCall(Texture=@n1, UVControl_RGB=@n2) {materialFunction=&FXCommon_TextureSample}.R
    B=ScalarParameter() {defaultValue=1.2, parameterName=TrailTextureMultiply}
   ) {constB=1.2})
  n4 = TextureCoordinate()
  n5 = ComponentMask(None=@n4) {r=true}
  n6 = Multiply(A=Abs(None=Subtract(A=@n5) {constB=0.5})) {constB=18}
  n7 = ComponentMask(None=@n4) {g=true}
  n8 = Saturate(None=@n7)
  n9 = Saturate(None=Multiply(A=OneMinus(None=@n6), B=@n8))
  n10 = VectorParameter() {parameterName=TrailSecondary_Color, defaultValue=<rgba 0.4319 0 1 1>}
  n11 = VectorParameter() {parameterName=TrailMain_Color, defaultValue=<rgba 0.1283 5 2.6651 1>}
  n12 = Power(Base=@n8) {constExponent=5}
  n13 = Multiply(A=ParticleColor().RGB, B=Add(A=Multiply(A=@n9) {constB=5}, B=LinearInterpolate(A=@n10.RGB, B=@n11.RGB, Alpha=@n12)))
  n14 = VertexColor()
  MaterialFunctionCall(Emissive=LinearInterpolate(A=Add(A=@n3, B=@n13), B=VectorParameter() {parameterName=Arrow_Color, defaultValue=<rgba 0.7635 3 2.7307 1>}.RGB, Alpha=@n14.R)) {materialFunction=&MF_FX_EmissiveAdjustments}
  n15 = Multiply(
   A=Constant3Vector() {constant=<rgba 0 0 1 0>}
   B=DynamicParameter() {defaultValue=<rgba 120 1 1 1>, paramNames [TrailZOffset,Param2,Param3,Param4]}.TrailZOffset
  ) {constB=120}
  n16 = OneMinus(None=ComponentMask(None=TextureCoordinate()) {g=true})
  n17 = DynamicParameter() {parameterIndex=1, defaultValue=<rgba 0 0 0 0>, paramNames [ObjectBoundOffset, Param2, Param3, Param4]}
  n18 = Multiply(
   A=Subtract(
    A=WorldPosition() {worldPositionShaderOffset=WPT_ExcludeAllShaderOffsets}.XYZ
    B=MaterialFunctionCall() {materialFunction=&ObjectPivotPoint}.Object Pivot Location
   )
   B=@n17.ObjectBoundOffset
  )
  Multiply(A=@n14.B, B=Add(A=Multiply(A=@n15, B=@n16), B=Multiply(A=@n16, B=@n18)))
  n19 = ComponentMask(None=@n4) {r=true}
  n20 = Multiply(A=Abs(None=Subtract(A=@n19) {constB=0.5})) {constB=2}
  n21 = Power(Base=Multiply(A=@n7, B=OneMinus(None=@n20)))
  n22 = Power(Base=OneMinus(None=@n7)) {constExponent=0.6}
  n23 = TextureCoordinate()
  n24 = Time()
  n25 = Constant2Vector() {g=1.25, r=0.25}
  n26 = TextureSample(UVs=Add(A=Multiply(A=@n23) {constB=0.6}, B=Multiply(A=@n24, B=@n25))) {samplerType=SAMPLERTYPE_LinearColor, texture=/Game/TimeOfDay/Skydome/PackedEclipseNoise_M}
  n27 = TextureCoordinate()
  n28 = Time()
  n29 = Constant2Vector() {g=0.8, r=0.15}
  n30 = TextureSample(UVs=Add(A=Multiply(A=@n27), B=Multiply(A=@n28, B=@n29))) {samplerType=SAMPLERTYPE_LinearColor, texture=&PackedEclipseNoise_M}
  n31 = Multiply(A=Multiply(A=Multiply(A=@n21, B=@n22), B=Min(A=@n26.R, B=@n30.B)), B=ScalarParameter() {defaultValue=8, parameterName=LargeTrailMultiply}) {constB=8}
  n32 = Multiply(A=@n21) {constB=0.3}
  n33 = Add(A=Saturate(None=Add(A=@n31, B=@n32)), B=@n9)
  Multiply(A=Saturate(None=Add(A=@n14.R, B=@n33))) {constB=1.25}
 ]
}
