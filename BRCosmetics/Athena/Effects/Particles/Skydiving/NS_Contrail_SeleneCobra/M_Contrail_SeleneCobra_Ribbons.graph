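# Unlit, translucent, two-sided material for ribbon and mesh particles in the SeleneCobra skydiving contrail Niagara effect.
# A soft smoke-trail texture is tiled and scrolled along U over time.
# A mask built from U is subtracted from it, eroding the smoke sharply near the start and gradually toward the far end.
# A power shapes the result, which is multiplied by particle alpha.
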Material /BRCosmetics/Athena/Effects/Particles/Skydiving/NS_Contrail_SeleneCobra/M_Contrail_SeleneCobra_Ribbons
graph {
 properties {
  bUsedWithNiagaraMeshParticles=true
  bUsedWithNiagaraRibbons=true
  blendMode=BLEND_Translucent
  pixelDepthOffsetMode=PDOM_Legacy
  previewMesh=&/SM_Contrail_SeleneCobra_Ribbon
  shadingModel=MSM_Unlit
  twoSided=true
  usedShadingModels=Unlit
 }
 nodes [
  n1 = TextureCoordinate()
  n2 = Constant2Vector() {g=1, r=0.8}
  n3 = Time()
  n4 = Constant2Vector() {r=-2}
  n5 = TextureSample(UVs=Add(A=Multiply(A=@n1, B=@n2) {constB=2}, B=Multiply(A=@n3, B=@n4))) {samplerType=SAMPLERTYPE_Grayscale, texture=&T_Smoke_Trail_Soft}
  n6 = ComponentMask(None=@n1) {r=true}
  n7 = Subtract(A=OneMinus(None=@n6)) {constB=0.8}
  n8 = Add(A=Saturate(None=Subtract(A=@n6) {constB=0.4}), B=Saturate(None=Multiply(A=@n7) {constB=5}))
  Multiply(A=ParticleColor().A, B=Power(Base=Subtract(A=@n5.R, B=@n8)))
 ]
}
