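# Unlit, two-sided translucent material for Niagara mesh particles that draws the arrow trail in the SeleneCobra skydiving contrail.
# An arrow texture, offset in U by Dynamic Param1, blends toward particle color over a gradient from a scrolling trail texture scaled by Param2.
# A sand texture adds extra brightness.
# Alpha combines arrow and trail masks, particle alpha and a fade at both U ends.
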Material /BRCosmetics/Athena/Effects/Particles/Skydiving/NS_Contrail_SeleneCobra/M_TrailArrow
graph {
 properties {
  bUsedWithNiagaraMeshParticles=true
  blendMode=BLEND_Translucent
  pixelDepthOffsetMode=PDOM_Legacy
  shadingModel=MSM_Unlit
  twoSided=true
  usedShadingModels=Unlit
 }
 nodes [
  n1 = TextureObjectParameter() {
   group=SandA
   parameterName=SandA_Texture
   samplerType=SAMPLERTYPE_Normal
   sortPriority=0
   texture=/Game/Tools/WaterfallGenerator/Art/T_Waterfall_Sand
  }
  n2 = MaterialFunctionCall(Texture=@n1, UVControl_RGB=Constant3Vector() {constant=<rgba 2 0.25 1 0>}) {materialFunction=&FXCommon_TextureSample}
  n3 = Constant3Vector() {constant=<rgba 0.2733 0 2 1>}
  n4 = Constant3Vector() {constant=<rgba 0.3851 1 0.9801 1>}
  n5 = Time() {bOverride_Period=true, period=1}
  n6 = Constant2Vector() {r=-1}
  n7 = TextureSample(UVs=Add(A=TextureCoordinate(), B=Multiply(A=@n5, B=@n6) {constB=-1})) {samplerType=SAMPLERTYPE_LinearColor, texture=/BRCosmetics/Athena/Effects/Particles/Skydiving/NS_Contrail_CubeNinja/T_CubeNinja_Trail}
  n8 = DynamicParameter() {defaultValue=<rgba 0 1 1 1>}
  n9 = Saturate(None=Multiply(A=Multiply(A=@n7.R, B=@n8.Param2)) {constB=2})
  n10 = Constant3Vector() {constant=<rgba 0.0418 1 0.8096 1>}
  n11 = ParticleColor()
  n12 = TextureCoordinate()
  n13 = Constant()
  n14 = TextureSample(UVs=Add(A=@n12, B=AppendVector(A=@n8.Param1, B=@n13))) {texture=&/T_Contrail_SeleneCobra_TrailArrow}
  n15 = Power(Base=@n14.R)
  Add(
   A=Saturate(None=Multiply(A=@n2.R) {constB=1.2})
   B=LinearInterpolate(
    A=LinearInterpolate(A=@n3, B=@n4, Alpha=@n9)
    B=LinearInterpolate(A=@n10, B=@n11.RGB, Alpha=@n15)
    Alpha=SmoothStep(Value=@n14.R) {constMax=0.1, constMin=0.06}
   )
  )
  Add(A=@n5, B=Multiply(A=ParticleRandom()) {constB=2})
  n16 = Multiply(A=@n9) {constB=0.25}
  n17 = ComponentMask(None=@n12) {r=true}
  n18 = OneMinus(None=@n17)
  n19 = Multiply(A=Power(Base=Multiply(A=@n18, B=@n17))) {constB=5}
  Multiply(A=@n11.A, B=Multiply(A=Add(A=@n16, B=@n14.R), B=Saturate(None=@n19)))
  Sine(None=Time())
 ]
}
