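# Cosmetic Niagara skydiving contrail (SeleneCobra) that dresses a character's hands and feet in cyan-to-purple arrows, glows, trails, smoke and debris.
# Most emitters use a two-way spawn group to pick sockets on two skeletal mesh inputs, apparently hands and feet.
# ParaGlideLeanAlpha drives orbiting ring radius, plus chunk spin and spawn rate; the GPU chunk ring has a lower-rate CPU variant.
# Falling arrows feed a secondary sprite emitter through particle reads.
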NiagaraSystem /BRCosmetics/Athena/Effects/Particles/Skydiving/NS_Contrail_SeleneCobra/NS_Contrail_SeleneCobra
dependencies {
 usedDataInterfaces [
  &NiagaraDataInterfaceColorCurve
  &NiagaraDataInterfaceCurve
  &NiagaraDataInterfaceMeshRendererInfo
  &NiagaraDataInterfaceArrayFloat
  &NiagaraDataInterfaceVectorCurve
  &NiagaraDataInterfaceParticleRead
 ]
 usedDynamicInputs [&SimulationPosition, &SetBoolByFloatComparison]
 usedModules [
  &SystemState
  &self:SystemSpawnScript.SystemScriptSource.SystemScriptGraph.NiagaraNodeAssignment_0.SetVariables_#1
 ]
}
emitters {
 Chunks {
  emitterUpdateScript [
   &EmitterState()
   &SpawnRate(SpawnRate=15, Use Spawn Probability=false)
  ]
  particleSpawnScript [
   &FN_SetSpawnGroup(Number of Spawn Groups=2)
   &InitializeParticle(
    Lifetime Max=1
    Lifetime Min=0.1
    Lifetime Mode=::NewEnumerator1
    Mesh Renderer Array Visibility Mode=::NewEnumerator3
    Mesh Renderer Info=DataInterface() {propertyValues {MeshRenderer=&self:Chunks_2.NiagaraMeshRendererProperties_0}}
    Mesh Scale Mode=::NewEnumerator1
    Mesh Uniform Scale Max=1.25
    Mesh Uniform Scale Min=0.5
    Sprite Size Mode=::NewEnumerator1
    Uniform Sprite Size Max=0.8
    Uniform Sprite Size Min=0.35
    UsePositionOffset=false
    Write Material Random=true
   )
   &MeshRotationForce(Rotation=&RandomRangeVector(Minimum=<5 5 5>, Maximum=<-5 -5 -5>))
   &ShapeLocation(MaskSpawnGroup=false, Override Seed=false, Sphere Radius=4, UseOwnerScale=true)
   &AddVelocity(Velocity=<0 0 100>)
   &InitialMeshOrientation(Use Rotation Vector=false)
   &ApplyInitialForces()
   &SkeletalMeshLocation(
    Bone / Socket Index=$Particles.Initial.SpawnGroup
    Bone Sampling Mode=::NewEnumerator4
    MaskSpawnGroup=false
    Sampled Position Offset=$Particles.Initial.ShapeLocation.ShapePosition
    Skeletal Mesh=$User.SkeletalMesh2
   )
  ]
  particleUpdateScript [
   &ParticleState()
   &ScaleMeshSize(Scale Factor=&VectorFromFloat(Value=&FloatFromCurve(FloatCurve=DataInterface() {propertyValues {Curve {keys [<at 0.4969 0.5588>, {tangentMode=RCTM_User, time=0.6, value=1.05}, <at 0.867 0.7441>, {tangentMode=RCTM_User, time=1, value=0}]}}})))
   &Drag(Drag=1.25, Use Linear Drag=true, Use Rotational Drag=true)
   &SolveForcesAndVelocity(Clamp Velocity=false, Limit Acceleration=false)
   &ScaleSpriteSize(Uniform Curve Sprite Scale=DataInterface() {propertyValues {Curve {keys [{tangentMode=RCTM_User, time=0, value=1}, <at 0.7 0.75>, {tangentMode=RCTM_User, time=1, value=0}]}}})
   &ScaleColor(
    Linear Color Curve=DataInterface() {propertyValues {
      AlphaCurve {keys [<at 0.0214 0>,<at 0.1454 1>,<at 0.6565 1>,<at 1 0>]}
      BlueCurve {keys [<at 0 1.0163>,<at 0.3591 1>,<at 0.9897 3>]}
      GreenCurve {keys [<at 0 1.25>,<at 0.3591 0.1612>,<at 0.9897 0>]}
      RedCurve {keys [<at 0 0>,<at 0.3591 0.1869>,<at 0.9897 1.2376>]}
     }}
    Scale Mode=::NewEnumerator2
   )
  ]
  rendererClasses [&NiagaraMeshRendererProperties]
  renderers [{rendererClass=&NiagaraMeshRendererProperties, rendererIndex=0}]
 }
 ChunksRing {
  simTarget=GPUComputeSim
  emitterUpdateScript [
   &EmitterState()
   &SpawnRate(SpawnRate=&Lerp_Float(A=55, B=60, Alpha=$User.ParaGlideLeanAlpha), Use Spawn Probability=false)
  ]
  particleSpawnScript [
   &InitializeParticle(
    Color=<rgba 0 1 0.8309 0.2>
    Lifetime Mode=::NewEnumerator1
    Mesh Scale Mode=::NewEnumerator1
    Mesh Uniform Scale Max:NiagaraFloat=&RandomRangeFloat(
     Minimum=0.5
     Maximum=&SelectFloatFromArray(
      Array Sampling Mode=::NewEnumerator0
      Direct Array Index=?
      Float Selection Array=DataInterface() {propertyValues {FloatData [0.8,2]}}
      Random Seed=0
     )
    )
    Mesh Uniform Scale Min=0.3
    Sprite Size Mode=::NewEnumerator3
    Uniform Sprite Size=12
    UsePositionOffset=false
    Write Material Random=true
   )
   &FN_SetSpawnGroup(Number of Spawn Groups=2)
   &InitialMeshOrientation(Use Rotation Vector=false)
  ]
  particleUpdateScript [
   &ParticleState()
   &ScaleSpriteSize(Uniform Curve Sprite Scale=DataInterface() {propertyValues {Curve {keys [
        {arriveTangent=9.0725, interpMode=RCIM_Cubic, leaveTangent=9.0725, tangentMode=RCTM_User, time=0, value=0}
        {arriveTangent=-0.0451, interpMode=RCIM_Cubic, leaveTangent=-0.0451, tangentMode=RCTM_User, time=0.1408, value=0.9847}
        {arriveTangent=-2.352, interpMode=RCIM_Cubic, leaveTangent=-2.352, tangentMode=RCTM_User, time=1, value=0}
       ]}}})
   &ScaleMeshSize(Scale Factor=&VectorFromFloat(Value=&FloatFromCurve(FloatCurve=DataInterface() {propertyValues {Curve {keys [
          {arriveTangent=9.0725, interpMode=RCIM_Cubic, leaveTangent=9.0725, tangentMode=RCTM_User, time=0, value=0}
          {arriveTangent=-0.0451, interpMode=RCIM_Cubic, leaveTangent=-0.0451, tangentMode=RCTM_User, time=0.1408, value=0.9847}
          {arriveTangent=-2.352, interpMode=RCIM_Cubic, leaveTangent=-2.352, tangentMode=RCTM_User, time=1, value=0}
         ]}}})))
   &SkeletalMeshLocation(Bone / Socket Index=$Particles.SpawnGroup, Bone Sampling Mode=::NewEnumerator4, MaskSpawnGroup=false, Skeletal Mesh=$User.SkeletalMesh)
   &RotateAroundPoint(
    Derive the Rotation Axis=false
    Radius:NiagaraFloat=&Multiply_Float(
     A=&RandomRangeFloat(Minimum=20, Maximum=25)
     B=&Lerp_Float(A=1, B=1.6, Alpha=$User.ParaGlideLeanAlpha)
    )
    Rotation Axis:Vector3f=&RandomRangeVector(Minimum=<-0.3 0 1>, Maximum=<0.3 0 1>)
    Rotation Center:NiagaraPosition=&AddVectorToPosition(Position=$Particles.SkeletalMeshLocation.SampledPosition, Vector=<0 0 80>)
    Rotation Phase:NiagaraFloat=&RandomRangeFloat(Minimum=0, Maximum=1)
    Rotation Rate:NiagaraFloat=&Multiply_Float(
     A=&RandomRangeFloat(Minimum=0.5, Maximum=1)
     B=&Multiply_Float(
      A=&Lerp_Float(A=0.6, B=1.5, Alpha=$User.ParaGlideLeanAlpha)
      B=&MakeCustomFloatFromBool(
       Bool=&SetBoolByIntComparison(A=$Particles.SpawnGroup, B=1, Comparison Type=::NewEnumerator6)
       True Float=-1
       False Float=1
      )
     )
    )
   )
  ]
  rendererClasses [&NiagaraMeshRendererProperties, &NiagaraSpriteRendererProperties]
  renderers [
   {rendererClass=&NiagaraMeshRendererProperties, rendererIndex=0}
   {rendererClass=&NiagaraSpriteRendererProperties, rendererIndex=1}
  ]
 }
 ChunksRingLOW {
  emitterUpdateScript [
   &EmitterState()
   &SpawnRate(SpawnRate=&Lerp_Float(A=30, B=40, Alpha=$User.ParaGlideLeanAlpha), Use Spawn Probability=false)
  ]
  particleSpawnScript [
   &InitializeParticle(
    Color=<rgba 0 1 0.8309 0.2>
    Lifetime Mode=::NewEnumerator1
    Mesh Scale Mode=::NewEnumerator1
    Mesh Uniform Scale Max:NiagaraFloat=&RandomRangeFloat(
     Minimum=0.5
     Maximum=&SelectFloatFromArray(
      Array Sampling Mode=::NewEnumerator0
      Direct Array Index=?
      Float Selection Array=DataInterface() {propertyValues {FloatData [0.8,2]}}
      Random Seed=0
     )
    )
    Mesh Uniform Scale Min=0.3
    Sprite Size Mode=::NewEnumerator3
    Uniform Sprite Size=20
    UsePositionOffset=false
    Write Material Random=true
   )
   &FN_SetSpawnGroup(Number of Spawn Groups=2)
   &InitialMeshOrientation(Use Rotation Vector=false)
  ]
  particleUpdateScript [
   &ParticleState()
   &ScaleSpriteSize(Uniform Curve Sprite Scale=DataInterface() {propertyValues {Curve {keys [
        {arriveTangent=9.0725, interpMode=RCIM_Cubic, leaveTangent=9.0725, tangentMode=RCTM_User, time=0, value=0}
        {arriveTangent=-0.0451, interpMode=RCIM_Cubic, leaveTangent=-0.0451, tangentMode=RCTM_User, time=0.1408, value=0.9847}
        {arriveTangent=-2.352, interpMode=RCIM_Cubic, leaveTangent=-2.352, tangentMode=RCTM_User, time=1, value=0}
       ]}}})
   &ScaleMeshSize(Scale Factor=&VectorFromFloat(Value=&FloatFromCurve(FloatCurve=DataInterface() {propertyValues {Curve {keys [
          {arriveTangent=9.0725, interpMode=RCIM_Cubic, leaveTangent=9.0725, tangentMode=RCTM_User, time=0, value=0}
          {arriveTangent=-0.0451, interpMode=RCIM_Cubic, leaveTangent=-0.0451, tangentMode=RCTM_User, time=0.1408, value=0.9847}
          {arriveTangent=-2.352, interpMode=RCIM_Cubic, leaveTangent=-2.352, tangentMode=RCTM_User, time=1, value=0}
         ]}}})))
   &SkeletalMeshLocation(Bone / Socket Index=$Particles.SpawnGroup, Bone Sampling Mode=::NewEnumerator4, MaskSpawnGroup=false, Skeletal Mesh=$User.SkeletalMesh)
   &RotateAroundPoint(
    Derive the Rotation Axis=false
    Radius:NiagaraFloat=&Multiply_Float(
     A=&RandomRangeFloat(Minimum=20, Maximum=25)
     B=&Lerp_Float(A=1, B=1.5, Alpha=$User.ParaGlideLeanAlpha)
    )
    Rotation Axis:Vector3f=&RandomRangeVector(Minimum=<-0.3 0 1>, Maximum=<0.3 0 1>)
    Rotation Center:NiagaraPosition=&AddVectorToPosition(Position=$Particles.SkeletalMeshLocation.SampledPosition, Vector=<0 0 80>)
    Rotation Phase:NiagaraFloat=&RandomRangeFloat(Minimum=0, Maximum=1)
    Rotation Rate:NiagaraFloat=&Multiply_Float(
     A=&RandomRangeFloat(Minimum=0.5, Maximum=1)
     B=&Multiply_Float(
      A=&Lerp_Float(A=0.6, B=1.5, Alpha=$User.ParaGlideLeanAlpha)
      B=&MakeCustomFloatFromBool(
       Bool=&SetBoolByIntComparison(A=$Particles.SpawnGroup, B=1, Comparison Type=::NewEnumerator6)
       True Float=-1
       False Float=1
      )
     )
    )
   )
  ]
  rendererClasses [&NiagaraMeshRendererProperties, &NiagaraSpriteRendererProperties]
  renderers [
   {rendererClass=&NiagaraMeshRendererProperties, rendererIndex=0}
   {rendererClass=&NiagaraSpriteRendererProperties, rendererIndex=1}
  ]
 }
 CosmicSmoke {
  emitterUpdateScript [
   &EmitterState()
   &SpawnRate(SpawnRate=10, Use Spawn Probability=false)
  ]
  particleSpawnScript [
   &FN_SetSpawnGroup(Number of Spawn Groups=2)
   &InitializeParticle(
    Lifetime Max=1
    Lifetime Min=0.45
    Lifetime Mode=::NewEnumerator1
    Sprite Size Mode=::NewEnumerator1
    Uniform Sprite Size Max=60
    Uniform Sprite Size Min=35
    UsePositionOffset=false
    Write Material Random=true
   )
   &ShapeLocation(MaskSpawnGroup=false, Override Seed=false, Sphere Radius=14, UseOwnerScale=true)
   &AddVelocity(Velocity=<0 0 350>)
   &SkeletalMeshLocation(
    Bone / Socket Index=$Particles.Initial.SpawnGroup
    Bone Sampling Mode=::NewEnumerator4
    MaskSpawnGroup=false
    Sampled Position Offset=&AddVectorToPosition(Position=$Particles.ShapeLocation.ShapePosition, Vector=<0 0 20>)
    Sampled Position Offset Coordinate Space:ENiagaraSkelMeshTransforms=::NewEnumerator0
    Skeletal Mesh=$User.SkeletalMesh
   )
  ]
  particleUpdateScript [
   &ParticleState()
   &Drag(Drag=3, Use Linear Drag=true, Use Rotational Drag=true)
   &SolveForcesAndVelocity(Clamp Velocity=false, Limit Acceleration=false)
   &ScaleSpriteSizeBySpeed(Max Scale Factor=<2 3>, Min Scale Factor=<1 2>)
   &ScaleSpriteSize(Uniform Curve Sprite Scale=DataInterface() {propertyValues {Curve {keys [{tangentMode=RCTM_User, time=0, value=0.6}, <at 0.4 0.75>, {tangentMode=RCTM_User, time=1, value=1.2}]}}})
   &ScaleColor(
    Linear Color Curve=DataInterface() {propertyValues {
      AlphaCurve {keys [<at 0.0651 0>,<at 0.2516 0.15>,<at 0.6565 0.08>,<at 1 0>]}
      BlueCurve {keys [<at 0 1.0163>,<at 0.3591 1>,<at 0.9897 3>]}
      GreenCurve {keys [<at 0 1.25>,<at 0.3591 0.1612>,<at 0.9897 0>]}
      RedCurve {keys [<at 0 0>,<at 0.3591 0.1869>,<at 0.9897 1.2376>]}
     }}
    Scale Mode=::NewEnumerator2
   )
  ]
  rendererClasses [&NiagaraSpriteRendererProperties]
  renderers [{rendererClass=&NiagaraSpriteRendererProperties}]
 }
 Dots {
  emitterUpdateScript [
   &EmitterState()
   &SpawnRate(SpawnRate=20, Use Spawn Probability=false)
  ]
  particleSpawnScript [
   &FN_SetSpawnGroup(Number of Spawn Groups=2)
   &InitializeParticle(
    Lifetime Max=0.7
    Lifetime Min=0.25
    Lifetime Mode=::NewEnumerator1
    Sprite Size Mode=::NewEnumerator1
    Uniform Sprite Size Max=1.2
    Uniform Sprite Size Min=0.5
    UsePositionOffset=false
    Write Material Random=true
   )
   &ShapeLocation(MaskSpawnGroup=false, Override Seed=false, Sphere Radius=10, UseOwnerScale=true)
   &AddVelocity(Velocity=<0 0 100>)
   &SkeletalMeshLocation(
    Bone / Socket Index=$Particles.Initial.SpawnGroup
    Bone Sampling Mode=::NewEnumerator4
    MaskSpawnGroup=false
    Sampled Position Offset=$Particles.Initial.ShapeLocation.ShapePosition
    Skeletal Mesh=$User.SkeletalMesh2
   )
  ]
  particleUpdateScript [
   &ParticleState()
   &SolveForcesAndVelocity(Clamp Velocity=false, Limit Acceleration=false)
   &ScaleSpriteSizeBySpeed(Max Scale Factor=<1.5 10>, Min Scale Factor=<0.5 5>)
   &ScaleSpriteSize(Uniform Curve Sprite Scale=DataInterface() {propertyValues {Curve {keys [{tangentMode=RCTM_User, time=0, value=1}, <at 0.7 0.75>, {tangentMode=RCTM_User, time=1, value=0}]}}})
   &ScaleColor(
    Linear Color Curve=DataInterface() {propertyValues {
      AlphaCurve {keys [<at 0.0214 0>,<at 0.1454 1>,<at 0.6565 1>,<at 1 0>]}
      BlueCurve {keys [<at 0 1.0163>,<at 0.3591 1>,<at 0.9897 3>]}
      GreenCurve {keys [<at 0 1.25>,<at 0.3591 0.1612>,<at 0.9897 0>]}
      RedCurve {keys [<at 0 0>,<at 0.3591 0.1869>,<at 0.9897 1.2376>]}
     }}
    Scale Mode=::NewEnumerator2
   )
   /SetVariables(Particles.ShadowColor:LinearColor=&HueShiftLinearColor(Hue=0, Saturation=1, Value=0.8, Linear Color=$Particles.Color)) {isSetParameters=true, module=SetVariables_#2}
   /SetVariables(Particles.ShadowSize:Vector2f=&Multiply_Vector2D(
     A=$Particles.SpriteSize
     B=&Vector2DFromFloat(Value=&FloatFromCurve(
       FloatCurve=DataInterface() {propertyValues {Curve {keys [
           {interpMode=RCIM_Cubic, leaveTangent=4.5862, tangentMode=RCTM_Break, time=0, value=0.4}
           {interpMode=RCIM_Cubic, leaveTangent=-1.285, tangentMode=RCTM_Break, time=0.4, value=1.1}
           {arriveTangent=-3.201, interpMode=RCIM_Cubic, tangentMode=RCTM_Break, time=1, value=0}
          ]}}}
       Scale Curve=5
      ))
    )) {isSetParameters=true, module=SetVariables_#3}
  ]
  rendererClasses [&NiagaraSpriteRendererProperties]
  renderers [
   {rendererClass=&NiagaraSpriteRendererProperties}
   {rendererClass=&NiagaraSpriteRendererProperties, rendererIndex=1}
  ]
 }
 Empty001 {
  emitterUpdateScript [
   &EmitterState()
   &SpawnRate(SpawnRate=32, Use Spawn Probability=false)
  ]
  particleSpawnScript [&InitializeParticle(
    Color=<rgba 0.1438 0 1 1>
    Lifetime=0.4
    Ribbon Width=20
    Ribbon Width Mode=::NewEnumerator1
    UsePositionOffset=false
    Write Material Random=true
   )]
  particleUpdateScript [
   &ParticleState()
   &ScaleRibbonWidth(Ribbon Width Scale=&FloatFromCurve(FloatCurve=DataInterface() {propertyValues {Curve {keys [<at 0.0105 0.5324>,<at 0.1077 0.9765>,<at 1 0.15>]}}}))
  ]
  rendererClasses [&NiagaraRibbonRendererProperties]
  renderers [{rendererClass=&NiagaraRibbonRendererProperties}]
 }
 Empty003 {
  emitterUpdateScript [
   &EmitterState()
   &SpawnParticlesFromOtherEmitter(
    Attribute Reader=DataInterface() {propertyValues {EmitterBinding {bindingMode=Other, emitterName=FallingArrows}}}
    Spawn Rate=55
    Use Maximum Age=false
    Use Minimum Age=false
   )
  ]
  particleSpawnScript [
   &FN_SetSpawnGroup(Number of Spawn Groups=2)
   &InitializeParticle(
    Color=<rgba 2.2838 1.2089 3 1>
    Lifetime Max=0.5
    Lifetime Min=0.2
    Lifetime Mode=::NewEnumerator1
    Position Offset=<0 0 35>
    Sprite Size Mode=::NewEnumerator1
    Uniform Sprite Size Max=&RandomRangeFloat(
     Minimum=8
     Maximum=&SelectFloatFromArray(
      Array Sampling Mode=::NewEnumerator0
      Direct Array Index=?
      Float Selection Array=DataInterface() {propertyValues {FloatData [12,35]}}
      Random Seed=0
     )
    )
    Uniform Sprite Size Min=2
    UsePositionOffset=true
    Write Material Random=true
   )
   &SampleParticlesFromOtherEmitter()
   &ShapeLocation(
    Disc Coverage=0.4643
    MaskSpawnGroup=false
    Override Seed=false
    Ring Radius=15
    Rotation Mode=::NewEnumerator2
    Shape Primitive=::NewEnumerator4
    UseOwnerScale=true
   )
   &SkeletalMeshLocation(
    Bone / Socket Index=$Particles.SpawnGroup
    Bone Sampling Mode=::NewEnumerator4
    MaskSpawnGroup=false
    Orientation Sampling=::NewEnumerator1
    Sampled Position Offset=&AddVectorToPosition(Position=$Particles.ShapeLocation.ShapePosition, Vector=<0 0 35>)
    Sampled Position Offset Coordinate Space:ENiagaraSkelMeshTransforms=::NewEnumerator0
    Skeletal Mesh=$User.SkeletalMesh
   ) {enabled=false}
   &AddVelocity(Velocity=&RandomRangeVector(Minimum=<0 0 250>, Maximum=<0 0 1000>)) {enabled=false}
  ]
  particleUpdateScript [
   &ParticleState()
   &ScaleColor(
    Linear Color Curve=DataInterface() {propertyValues {
      AlphaCurve {keys [<at 0.0214 0>,<at 0.1454 0.8>,<at 0.6565 0.6>,<at 1 0>]}
      BlueCurve {keys [<at 0 6.5042>,<at 0.4467 5>,<at 0.9897 5>]}
      GreenCurve {keys [<at 0 8>,<at 0.4467 0.8058>,<at 0.9897 0>]}
      RedCurve {keys [<at 0 0>,<at 0.4467 0.9346>,<at 0.9897 2.0627>]}
     }}
    Scale Mode=::NewEnumerator2
   )
   &ScaleSpriteSize(Uniform Curve Sprite Scale=DataInterface() {propertyValues {Curve {keys [{tangentMode=RCTM_User, time=0, value=0}, <at 0.5 1>, <at 0.7 1>, {tangentMode=RCTM_User, time=1, value=0.15}]}}})
   &Drag(Drag=2, Use Linear Drag=true, Use Rotational Drag=true)
   &CurlNoiseForce(Noise Strength=200, Pan Noise=false)
   &SolveForcesAndVelocity(Clamp Velocity=false, Limit Acceleration=false)
  ]
  rendererClasses [&NiagaraSpriteRendererProperties]
  renderers [{rendererClass=&NiagaraSpriteRendererProperties}]
 }
 FallingArrows {
  emitterUpdateScript [
   &EmitterState()
   &SpawnRate(SpawnRate=2, Use Spawn Probability=false)
  ]
  particleSpawnScript [
   &InitializeParticle(
    Color Maximum=<rgba 0.1386 0.0343 1 0.5>
    Color Minimum=<rgba 0.1994 0.1886 1 0.5>
    Color Mode=::NewEnumerator2
    Lifetime Max=0.8
    Lifetime Min=0.25
    Lifetime Mode=::NewEnumerator1
    Position Offset=<0 0 300>
    UsePositionOffset=true
    Write Material Random=true
   )
   &ShapeLocation(MaskSpawnGroup=false, Override Seed=false, Ring Radius=90, Shape Primitive=::NewEnumerator4, UseOwnerScale=true)
   &InitialMeshOrientation(
    Mesh Orientation Mode=::NewEnumerator2
    Rotation=&RandomRangeVector(Minimum=<0 0 0>, Maximum=<0 0 1>)
    Use Rotation Vector=true
   )
   &AddVelocity(Velocity=&RandomRangeVector(Minimum=<0 0 -800>, Maximum=<0 0 -1200>))
   &DynamicMaterialParameters(
    Index 0 Param 1=140
    Module.Index0Param0Write=true
    Module.Index0Param1Write=true
    Module.Index0Param2Write=true
    Module.Index0Param3Write=true
   )
  ]
  particleUpdateScript [
   &ParticleState()
   &SolveForcesAndVelocity(Clamp Velocity=false, Limit Acceleration=false)
   &GenerateLocationEvent(Use Event Probability=false, UseEventDelay=false)
  ]
  rendererClasses [&NiagaraMeshRendererProperties]
  renderers [{rendererClass=&NiagaraMeshRendererProperties, rendererIndex=0}]
 }
 FootGlow {
  emitterUpdateScript [
   &EmitterState()
   &SpawnRate(SpawnRate=22, Use Spawn Probability=false)
  ]
  particleSpawnScript [
   &FN_SetSpawnGroup(Number of Spawn Groups=2)
   &InitializeParticle(
    Color=<rgba 0 1 1 0.1>
    Lifetime Max=0.8
    Lifetime Min=0.35
    Lifetime Mode=::NewEnumerator1
    Sprite Rotation Mode=::NewEnumerator2
    Sprite Size Mode=::NewEnumerator3
    Uniform Sprite Size=60
    UsePositionOffset=false
    Write Material Random=true
   )
  ]
  particleUpdateScript [
   &ParticleState()
   &SkeletalMeshLocation(
    Bone / Socket Index=$Particles.Initial.SpawnGroup
    Bone Sampling Mode=::NewEnumerator4
    MaskSpawnGroup=false
    Sampled Position Offset=<0 0 5>
    Sampled Position Offset Coordinate Space:ENiagaraSkelMeshTransforms=::NewEnumerator0
    Skeletal Mesh=$User.SkeletalMesh2
   )
   &SolveForcesAndVelocity(Clamp Velocity=false, Limit Acceleration=false)
   &CameraOffset(Camera Offset Amount=10)
   &ScaleColor(
    Curve Index=$System.LoopedAge
    Linear Color Curve=DataInterface() {propertyValues {
      AlphaCurve {keys [<at 0 1>,<at 0.1323 0.25>,<at 0.2516 1>,<at 0.373 0.5>,<at 0.4901 0.8>,<at 0.6007 0.2>,<at 0.72 1>,<at 0.8241 0.4>,<at 1 1>]}
      BlueCurve {keys [<at 0 1>,<at 1 1>]}
      GreenCurve {keys [<at 0 1>,<at 1 1>]}
      RedCurve {keys [<at 0 1>,<at 1 1>]}
     }}
    Scale Mode=::NewEnumerator2
   )
  ]
  rendererClasses [&NiagaraSpriteRendererProperties]
  renderers [
   {rendererClass=&NiagaraSpriteRendererProperties}
   {rendererClass=&NiagaraSpriteRendererProperties, rendererIndex=1}
  ]
 }
 HandFallingArrows {
  emitterUpdateScript [
   &EmitterState()
   &SpawnRate(SpawnRate=2.5, Use Spawn Probability=false)
  ]
  particleSpawnScript [
   &FN_SetSpawnGroup(Number of Spawn Groups=2)
   &InitializeParticle(
    Color=<rgba 1.9651 4 3.6398 1>
    Lifetime Max=3
    Lifetime Min=2
    Lifetime Mode=::NewEnumerator1
    UsePositionOffset=false
    Write Material Random=true
   )
   &InitialMeshOrientation(
    Mesh Orientation Mode=::NewEnumerator2
    Rotation=&RandomRangeVector(Minimum=<0 0 0>, Maximum=<0 0 1>)
    Use Rotation Vector=true
   )
   &ShapeLocation(MaskSpawnGroup=false, Override Seed=false, Ring Radius=3, Shape Primitive=::NewEnumerator4, UseOwnerScale=true)
  ]
  particleUpdateScript [
   &ParticleState()
   &DynamicMaterialParameters(
    Index 0 Param 1=&FloatFromCurve(FloatCurve=DataInterface() {propertyValues {Curve {keys [
         {interpMode=RCIM_Cubic, leaveTangent=-0.4044, tangentMode=RCTM_Break, time=0, value=-1}
         {arriveTangent=1.1601, tangentMode=RCTM_User, time=1, value=1}
        ]}}})
    Index 0 Param 2=&FloatFromCurve(FloatCurve=DataInterface() {propertyValues {Curve {keys [
         {arriveTangent=-5.2055, interpMode=RCIM_Cubic, tangentMode=RCTM_Break, time=0, value=-1}
         {arriveTangent=-0.3658, interpMode=RCIM_Cubic, leaveTangent=-4.0413, tangentMode=RCTM_Break, time=0.5947, value=0.875}
         {arriveTangent=-7.8799, tangentMode=RCTM_User, time=0.9051, value=-1}
        ]}}})
    Module.Index0Param0Write=true
    Module.Index0Param1Write=true
    Module.Index0Param2Write=true
    Module.Index0Param3Write=true
   )
   &SkeletalMeshLocation(
    Bone / Socket Index=$Particles.Initial.SpawnGroup
    Bone Sampling Mode=::NewEnumerator4
    MaskSpawnGroup=false
    Orientation Sampling=::NewEnumerator1
    Sampled Position Offset=&AddVectorToPosition(Position=$Particles.ShapeLocation.ShapePosition, Vector=<0 0 25>)
    Sampled Position Offset Coordinate Space:ENiagaraSkelMeshTransforms=::NewEnumerator0
    Skeletal Mesh=$User.SkeletalMesh
   )
   &ScaleColor(
    Linear Color Curve=DataInterface() {propertyValues {
      AlphaCurve {keys [<at 0 1>,<at 1 0.8>]}
      BlueCurve {keys [<at 0 8.1924>,<at 0.1692 5>,<at 0.4402 0.8899>]}
      GreenCurve {keys [<at 0 10>,<at 0.1692 4.0994>,<at 0.4402 1>]}
      RedCurve {keys [<at 0 2.0669>,<at 0.1692 0.0152>,<at 0.4402 0.3851>]}
     }}
    Scale Mode=::NewEnumerator2
   )
  ]
  rendererClasses [&NiagaraMeshRendererProperties]
  renderers [{rendererClass=&NiagaraMeshRendererProperties, rendererIndex=0}]
 }
 HandGlow {
  emitterUpdateScript [
   &EmitterState(Life Cycle Mode=::NewEnumerator1, Loop Behavior=::NewEnumerator1, Loop Duration=0, UseLoopDelay=false)
   &SpawnBurst_Instantaneous(Spawn Count=2, Spawn Probability=HlslExpression() {hlslExpression="Emitter.LoopCount == 0"}, Use Spawn Probability=true)
   &SpawnBurst_Instantaneous(Spawn Count=2, Spawn Time=0.9002, Use Spawn Probability=false, UseLoopCountLimit=false) {module=SpawnBurst_Instantaneous001}
  ]
  particleSpawnScript [
   &FN_SetSpawnGroup(Number of Spawn Groups=2)
   &InitializeParticle(
    Color=<rgba 0.1274 1 1 0.4>
    Mesh Scale Mode=::NewEnumerator3
    Mesh Uniform Scale=0.3
    Sprite Rotation Mode=::NewEnumerator2
    Sprite Size=<50 50>
    Sprite Size Mode=::NewEnumerator4
    UsePositionOffset=false
    Write Material Random=true
   ) {module=InitializeParticle001}
   &InitialMeshOrientation(Mesh Orientation Mode=::NewEnumerator2, Rotation=<-0.25 0 0>, Use Rotation Vector=true)
  ]
  particleUpdateScript [
   &ParticleState(Kill Particles When Lifetime Has Elapsed=false)
   &SkeletalMeshLocation(
    Bone / Socket Index=$Particles.SpawnGroup
    Bone Sampling Mode=::NewEnumerator4
    Inherit Velocity=false
    Interpolate Mesh Sampling=false
    MaskSpawnGroup=false
    Sampled Position Offset=<0 0 31.1868>
    Skeletal Mesh=$User.SkeletalMesh
   )
   &UpdateMeshOrientation(Rotation Rate=0.6, Rotation Vector=<0 -1 0>) {enabled=false}
   &ScaleColor(
    Curve Index=$System.LoopedAge
    Linear Color Curve=DataInterface() {propertyValues {
      AlphaCurve {keys [<at 0 1>,<at 0.1323 0.25>,<at 0.2516 1>,<at 0.373 0.5>,<at 0.4901 0.8>,<at 0.6007 0.2>,<at 0.72 1>,<at 0.8241 0.4>,<at 1 1>]}
      BlueCurve {keys [<at 0 1>,<at 1 1>]}
      GreenCurve {keys [<at 0 1>,<at 1 1>]}
      RedCurve {keys [<at 0 1>,<at 1 1>]}
     }}
    Scale Mode=::NewEnumerator2
   )
  ]
  rendererClasses [&NiagaraMeshRendererProperties, &NiagaraSpriteRendererProperties]
  renderers [
   {rendererClass=&NiagaraMeshRendererProperties, rendererIndex=0}
   {rendererClass=&NiagaraSpriteRendererProperties, rendererIndex=1}
   {rendererClass=&NiagaraSpriteRendererProperties, rendererIndex=2}
  ]
 }
 MainArrow {
  emitterUpdateScript [
   &EmitterState(Life Cycle Mode=::NewEnumerator1, Loop Behavior=::NewEnumerator1, Loop Duration=0, UseLoopDelay=false)
   &SpawnBurst_Instantaneous(Spawn Count=2, Spawn Probability=HlslExpression() {hlslExpression="Emitter.LoopCount == 0"}, Use Spawn Probability=true)
  ]
  particleSpawnScript [
   &FN_SetSpawnGroup(Number of Spawn Groups=2)
   &InitializeParticle(
    Color=<rgba 0.5214 0.4071 1 1>
    Mass Mode=::NewEnumerator1
    Mesh Scale Mode=::NewEnumerator3
    Mesh Uniform Scale=0.8
    UsePositionOffset=false
    Write Material Random=true
   ) {module=InitializeParticle001}
   &InitialMeshOrientation(Mesh Orientation Mode=::NewEnumerator2, Rotation Coordinate Space=::NewEnumerator0, Use Rotation Vector=true)
   /SetVariables(Particles.ArrowDynParam:Vector4f=<120 0 0 0>, Particles.Render2Color:LinearColor=<rgba 1 1 1 0.35>) {isSetParameters=true}
  ]
  particleUpdateScript [
   &ParticleState(Kill Particles When Lifetime Has Elapsed=false)
   &UpdateMeshOrientation(Rotation Rate=0.6, Rotation Vector=<0 0 1>) {enabled=false}
   &SkeletalMeshLocation(
    Bone / Socket Index=$Particles.SpawnGroup
    Bone Sampling Mode=::NewEnumerator4
    MaskSpawnGroup=false
    Orientation Sampling=::NewEnumerator1
    Sampled Position Offset=<0 0 20>
    Sampled Position Offset Coordinate Space:ENiagaraSkelMeshTransforms=::NewEnumerator0
    Skeletal Mesh=$User.SkeletalMesh
   )
  ]
  rendererClasses [&NiagaraMeshRendererProperties]
  renderers [
   {rendererClass=&NiagaraMeshRendererProperties, rendererIndex=0}
   {rendererClass=&NiagaraMeshRendererProperties, rendererIndex=1}
   {rendererClass=&NiagaraMeshRendererProperties, rendererIndex=2}
  ]
 }
 Ribbons {
  emitterUpdateScript [
   &EmitterState()
   &SpawnRate(SpawnRate=2, Use Spawn Probability=false)
  ]
  particleSpawnScript [
   &FN_SetSpawnGroup(Number of Spawn Groups=2)
   &PartitionParticles()
   &InitializeParticle(
    Color=<rgba 0.8203 0.4605 1 0.25>
    Lifetime Max=3
    Lifetime Min=2
    Lifetime Mode=::NewEnumerator1
    Mass Mode=::NewEnumerator1
    Mesh Scale Mode=::NewEnumerator3
    Sprite Size Mode=::NewEnumerator4
    UsePositionOffset=false
    Write Material Random=true
   ) {module=InitializeParticle001}
   &InitialMeshOrientation(
    Mesh Orientation Mode=::NewEnumerator2
    Rotation=&RandomRangeVector(Minimum=<0 0 0>, Maximum=<0 0 1>)
    Use Rotation Vector=true
   )
  ]
  particleUpdateScript [
   &ParticleState()
   &UpdateMeshOrientation(
    Rotation Rate=&SelectFloatFromArray(
     Direct Array Index=$Particles.Initial.PartitionParticles.Partition
     Float Selection Array=DataInterface() {propertyValues {FloatData [-1,1]}}
    )
    Rotation Vector=<0 0 1>
   )
   &SkeletalMeshLocation(
    Bone / Socket Index=$Particles.SpawnGroup
    Bone Sampling Mode=::NewEnumerator4
    Inherit Velocity=false
    Interpolate Mesh Sampling=false
    MaskSpawnGroup=false
    Sampled Position Offset=&VectorFromCurve(
     VectorCurve=DataInterface() {propertyValues {
       XCurve {keys [<at 0 0>,<at 1 0>]}
       YCurve {keys [<at 0 0>,<at 1 0>]}
       ZCurve {keys [
         {interpMode=RCIM_Cubic, leaveTangent=0.0639, tangentMode=RCTM_Break, time=0, value=0.55}
         {arriveTangent=1.8134, tangentMode=RCTM_User, time=0.9, value=1.35}
        ]}
      }}
     Scale Curve=<1 1 60>
    )
    Skeletal Mesh=$User.SkeletalMesh
   )
   &ScaleMeshSize(Scale Factor=&VectorFromCurve(VectorCurve=DataInterface() {propertyValues {
       XCurve {keys [<at 0 1>,<at 1 1>]}
       YCurve {keys [<at 0 1>,<at 1 1>]}
       ZCurve {keys [{interpMode=RCIM_Cubic, leaveTangent=2.7698, tangentMode=RCTM_Break, time=0, value=1}, <at 0.2 1.375>, <at 1 2>]}
      }}))
   &SolveForcesAndVelocity(Clamp Velocity=false, Limit Acceleration=false)
   &ScaleColor(
    Linear Color Curve=DataInterface() {propertyValues {
      AlphaCurve {keys [<at 0 0>,<at 0.2169 1>,<at 0.4598 1>,<at 1 0.2>]}
      BlueCurve {keys [<at 0.1005 4.0972>,<at 0.386 1>,<at 0.9897 3>]}
      GreenCurve {keys [<at 0.1005 5>,<at 0.386 0.1112>,<at 0.9897 0>]}
      RedCurve {keys [<at 0.1005 0.78>,<at 0.386 0.4056>,<at 0.9897 1.2376>]}
     }}
    Scale Mode=::NewEnumerator2
   ) {module=ScaleColor001}
  ]
  rendererClasses [&NiagaraMeshRendererProperties]
  renderers [{rendererClass=&NiagaraMeshRendererProperties, rendererIndex=0}]
 }
 RingTrail {
  emitterUpdateScript [
   &EmitterState()
   &SpawnRate(SpawnRate=64, Use Spawn Probability=false)
  ]
  particleSpawnScript [
   &InitializeParticle(
    Color=<rgba 0 2 1.1495 0.2>
    Lifetime=2.5
    Ribbon Width=12
    Ribbon Width Mode=::NewEnumerator1
    UsePositionOffset=false
    Write Material Random=true
   )
   &FN_SetSpawnGroup(Number of Spawn Groups=2)
  ]
  particleUpdateScript [
   &ParticleState()
   &SkeletalMeshLocation(Bone / Socket Index=$Particles.SpawnGroup, Bone Sampling Mode=::NewEnumerator4, MaskSpawnGroup=false, Skeletal Mesh=$User.SkeletalMesh)
   &RotateAroundPoint(
    Radius:NiagaraFloat=&Lerp_Float(A=22, B=38, Alpha=$User.ParaGlideLeanAlpha)
    Rotation Center:NiagaraPosition=&AddVectorToPosition(Position=$Particles.SkeletalMeshLocation.SampledPosition, Vector=<0 0 80>)
    Rotation Rate=-1
   )
   &ScaleColor(Scale Mode=::NewEnumerator2)
  ]
  rendererClasses [&NiagaraRibbonRendererProperties]
  renderers [{rendererClass=&NiagaraRibbonRendererProperties}]
 }
 SpeedDots {
  emitterUpdateScript [
   &EmitterState()
   &SpawnRate(SpawnRate=30, Use Spawn Probability=false)
  ]
  particleSpawnScript [
   &FN_SetSpawnGroup(Number of Spawn Groups=2)
   &InitializeParticle(
    Lifetime Max=0.8
    Lifetime Min=0.35
    Lifetime Mode=::NewEnumerator1
    Sprite Size Mode=::NewEnumerator1
    Uniform Sprite Size Max=1.2
    Uniform Sprite Size Min=0.5
    UsePositionOffset=false
    Write Material Random=true
   )
   &ShapeLocation(MaskSpawnGroup=false, Override Seed=false, Sphere Radius=14, UseOwnerScale=true)
   &AddVelocity(Velocity=<0 0 250>)
   &SkeletalMeshLocation(
    Bone / Socket Index=$Particles.Initial.SpawnGroup
    Bone Sampling Mode=::NewEnumerator4
    MaskSpawnGroup=false
    Sampled Position Offset=&AddVectorToPosition(Position=$Particles.ShapeLocation.ShapePosition, Vector=<0 0 20>)
    Sampled Position Offset Coordinate Space:ENiagaraSkelMeshTransforms=::NewEnumerator0
    Skeletal Mesh=$User.SkeletalMesh
   )
  ]
  particleUpdateScript [
   &ParticleState()
   &Drag(Use Linear Drag=true, Use Rotational Drag=true)
   &SolveForcesAndVelocity(Clamp Velocity=false, Limit Acceleration=false)
   &ScaleSpriteSizeBySpeed(Max Scale Factor=<1 15>, Min Scale Factor=<0.4 8>)
   &ScaleSpriteSize(Uniform Curve Sprite Scale=DataInterface() {propertyValues {Curve {keys [{tangentMode=RCTM_User, time=0, value=1}, <at 0.7 0.75>, {tangentMode=RCTM_User, time=1, value=0}]}}})
   &ScaleColor(
    Linear Color Curve=DataInterface() {propertyValues {
      AlphaCurve {keys [<at 0.0214 0>,<at 0.1454 1>,<at 0.6565 1>,<at 1 0>]}
      BlueCurve {keys [<at 0 1.0163>,<at 0.3591 1>,<at 0.9897 3>]}
      GreenCurve {keys [<at 0 1.25>,<at 0.3591 0.1612>,<at 0.9897 0>]}
      RedCurve {keys [<at 0 0>,<at 0.3591 0.1869>,<at 0.9897 1.2376>]}
     }}
    Scale Mode=::NewEnumerator2
   )
   /SetVariables(Particles.ShadowColor:LinearColor=&HueShiftLinearColor(Hue=0, Saturation=1, Value=0.8, Linear Color=$Particles.Color)) {isSetParameters=true, module=SetVariables_#2}
   /SetVariables(Particles.ShadowSize:Vector2f=&Multiply_Vector2D(
     A=$Particles.SpriteSize
     B=&Vector2DFromFloat(Value=&FloatFromCurve(
       FloatCurve=DataInterface() {propertyValues {Curve {keys [
           {interpMode=RCIM_Cubic, leaveTangent=4.5862, tangentMode=RCTM_Break, time=0, value=0.4}
           {interpMode=RCIM_Cubic, leaveTangent=-1.285, tangentMode=RCTM_Break, time=0.4, value=1.1}
           {arriveTangent=-3.201, interpMode=RCIM_Cubic, tangentMode=RCTM_Break, time=1, value=0}
          ]}}}
       Scale Curve=5
      ))
    )) {isSetParameters=true, module=SetVariables_#3}
  ]
  rendererClasses [&NiagaraSpriteRendererProperties]
  renderers [
   {rendererClass=&NiagaraSpriteRendererProperties}
   {rendererClass=&NiagaraSpriteRendererProperties, rendererIndex=1}
  ]
 }
 Trail {
  emitterUpdateScript [
   &EmitterState()
   &FN_SpawnRibbon_v2(Number of Ribbons=2, Spawn Mode=::NewEnumerator0, SpawnRate=64)
  ]
  particleSpawnScript [
   /Game/Effects/Niagara/Modules/Ribbons/FN_InitializeRibbon_v4(
    Enable Skeletal Position=true
    Lifetime=0.4
    Offset Axis:ENiagaraOrientationAxis=::ZAxis
    Ribbon Width=8
    SkeletalMesh:NiagaraDataInterfaceSkeletalMesh=$User.SkeletalMesh2
   )
   &FN_FlipRibbonV()
   &AddVelocity(Velocity=<0 0 250>)
  ]
  particleUpdateScript [
   &SolveForcesAndVelocity(Clamp Velocity=false, Limit Acceleration=false)
   /Game/Effects/Niagara/Modules/Ribbons/FN_UpdateRibbon_v4(Enable Skeletal Position=true, SkeletalMesh:NiagaraDataInterfaceSkeletalMesh=$User.SkeletalMesh2)
   &ScaleColor(
    Linear Color Curve=DataInterface() {propertyValues {
      AlphaCurve {keys [<at 0 0>,<at 0.0428 1>,<at 0.8348 1>,<at 1 0.6>]}
      BlueCurve {keys [<at 0.0043 3>,<at 0.1005 1>,<at 0.2973 1>,<at 0.9897 3>]}
      GreenCurve {keys [<at 0.0043 3>,<at 0.1005 1>,<at 0.2973 0.0713>,<at 0.9897 0.0985>]}
      RedCurve {keys [<at 0.0043 0>,<at 0.1005 0>,<at 0.2973 0.1779>,<at 0.9897 0.4783>]}
     }}
    Scale Mode=::NewEnumerator2
   )
   &ScaleRibbonWidth(Ribbon Width Scale=&FloatFromCurve(FloatCurve=DataInterface() {propertyValues {Curve {keys [
         <at 0 0.85>
         {arriveTangent=-0.5657, interpMode=RCIM_Cubic, leaveTangent=-0.3336, tangentMode=RCTM_Break, time=0.2, value=1}
         {arriveTangent=-1.6468, interpMode=RCIM_Cubic, leaveTangent=1.0861, tangentMode=RCTM_Break, time=1, value=0}
        ]}}}))
  ]
  rendererClasses [&NiagaraRibbonRendererProperties]
  renderers [{rendererClass=&NiagaraRibbonRendererProperties}]
 }
}
summary {
 systemName=NS_Contrail_SeleneCobra
 emitters [
  {emitterName=Trail, rendererClasses [&NiagaraRibbonRendererProperties]}
  {emitterName=Dots, rendererClasses [&NiagaraSpriteRendererProperties]}
  {emitterName=Chunks, rendererClasses [&NiagaraMeshRendererProperties]}
  {emitterName=MainArrow, rendererClasses [&NiagaraMeshRendererProperties]}
  {emitterName=Ribbons, rendererClasses [&NiagaraMeshRendererProperties]}
  {emitterName=SpeedDots, rendererClasses [&NiagaraSpriteRendererProperties]}
  {emitterName=HandGlow, rendererClasses [&NiagaraMeshRendererProperties, &NiagaraSpriteRendererProperties]}
  {emitterName=FootGlow, rendererClasses [&NiagaraSpriteRendererProperties]}
  {emitterName=HandFallingArrows, rendererClasses [&NiagaraMeshRendererProperties]}
  {emitterName=FallingArrows, rendererClasses [&NiagaraMeshRendererProperties]}
  {
   emitterName=ChunksRing
   simTarget=GPUComputeSim
   rendererClasses [&NiagaraMeshRendererProperties, &NiagaraSpriteRendererProperties]
  }
  {emitterName=CosmicSmoke, rendererClasses [&NiagaraSpriteRendererProperties]}
  {emitterName=Empty003, rendererClasses [&NiagaraSpriteRendererProperties]}
  {emitterName=Empty001, rendererClasses [&NiagaraRibbonRendererProperties]}
  {emitterName=RingTrail, rendererClasses [&NiagaraRibbonRendererProperties]}
  {emitterName=ChunksRingLOW, rendererClasses [&NiagaraMeshRendererProperties, &NiagaraSpriteRendererProperties]}
 ]
 userVariables [
  User.bIsFrontEnd {type=NiagaraBool}
  User.bIsFrontEndPreview {type=NiagaraBool}
  User.ParaGlideLeanAlpha {type=NiagaraFloat, defaultValue=0, description=""}
  User.RibbonIDOffset {type=NiagaraInt32}
  User.SkeletalMesh {
   type=NiagaraDataInterfaceSkeletalMesh
   defaultValue {
    struct=NiagaraExt_VariableValue_DataInterface
    value {dataInterface=&self:NiagaraDataInterfaceSkeletalMesh_0}
   }
  }
  User.SkeletalMesh2 {
   type=NiagaraDataInterfaceSkeletalMesh
   defaultValue {
    struct=NiagaraExt_VariableValue_DataInterface
    value {dataInterface=&self:NiagaraDataInterfaceSkeletalMesh_1}
   }
  }
 ]
}
